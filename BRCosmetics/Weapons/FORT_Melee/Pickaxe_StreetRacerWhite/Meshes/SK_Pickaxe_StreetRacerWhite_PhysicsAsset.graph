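# Physics asset for the StreetRacerWhite pickaxe skeletal mesh, a two-handed melee weapon cosmetic.
# It holds a single capsule body on the Melee_TwoHanded_Handle bone, offset along Z and rotated about 90 degrees.
# It defines no constraints, so the pickaxe has one rigid collision shape and no jointed parts.
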PhysicsAsset /BRCosmetics/Weapons/FORT_Melee/Pickaxe_StreetRacerWhite/Meshes/SK_Pickaxe_StreetRacerWhite_PhysicsAsset
bodies {Melee_TwoHanded_Handle [{shapeType=Capsule, length=73.6411, radius=23.9861, center=<0.9068 -0.9067 24.97>, rotation=<rot -0.1225 -90 -0.0388>}]}
constraints []
